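# Looping ambient spot cue for a smaller air-conditioning vent at the Helios Weather Station POI, in the Athena ambient sound class.
# A Random node picks one of three equally weighted AC duct air loops reused from Asteria's Industrial Town, then passes it through a Modulator.
# An Attenuation node referencing the cue's dedicated attenuation asset feeds the root Mixer, whose second input is empty.
# A concurrency override caps simultaneous instances at three.
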SoundCue /Game/Sounds/Helios/POI/WeatherStation/Spots/Spot_Aircon_Vent_Loop_Smaller_Cue
root=Mixer_1
nodes [
 {
  id=Attenuation_1
  children [Modulator_0]
  properties {
   attenuationSettings=&/Spot_Aircon_Vent_Loop_Smaller_Att
   attenuationOverrides {
    bAttenuateWithLPF=true
    bEnableFocusInterpolation=true
    bEnableListenerFocus=true
    bEnableOcclusion=true
    bEnableReverbSend=false
    falloffDistance=700
    focusReleaseInterpSpeed=0.5
    lPFFrequencyAtMax=2000
    lPFRadiusMax=3000
    lPFRadiusMin=500
    nonFocusVolumeAttenuation=0.5
    occlusionInterpolationTime=1
    occlusionLowPassFilterFrequency=2000
    occlusionTraceChannel=ECC_GameTraceChannel11
    occlusionVolumeAttenuation=0.5
    stereoSpread=512
    attenuationShapeExtents {x=250}
   }
  }
 }
 {id=Mixer_1, children [Attenuation_1, null], properties {inputVolume [1,1]}}
 {id=Modulator_0, children [Random_1]}
 {id=Random_1, children [WavePlayer_0,WavePlayer_1,WavePlayer_2], properties {weights [1,1,1]}}
 {id=WavePlayer_0, properties {bLooping=true, soundWaveAssetPtr=/Game/Sounds/Asteria/POI/IndustrialTown/Spots/Spot_AC_Duct_Air_03}}
 {id=WavePlayer_1, properties {bLooping=true, soundWaveAssetPtr=/Game/Sounds/Asteria/POI/IndustrialTown/Spots/Spot_AC_Duct_Air_01}}
 {id=WavePlayer_2, properties {bLooping=true, soundWaveAssetPtr=/Game/Sounds/Asteria/POI/IndustrialTown/Spots/Spot_AC_Duct_Air_02}}
]
properties {
 bOverrideConcurrency=true
 maxDistance=1500
 soundClassObject=&Fort_Ambient_Athena
 volumeMultiplier=0.4
 attenuationOverrides {attenuationShapeExtents {x=0}}
 concurrencyOverrides {maxCount=3}
}
